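; Pistol "mark execute" component for the Smartgun, tracking powerup stacks and targeting startup state on its owning FortPawn.
; Off the server, Tick maps stacks to a shrinking reticle size and passes it to TickAutoAim, whose body is empty here.
; Server-named custom events apply a no-lock-on spread GameplayEffect and zero both stack counters on reload.
; Targeting start records a timestamp for a one-second progress value; the replicated-stack OnRep only prints.
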(event EventBeginPlay)

(event Custom|TellServerReduceSpread
  (GameplayEffects|ApplyGameplayEffectToSelf :self (Ability|GetAbilitySystemComponent (Components|GetOwner)) :GameplayEffectClass "/Smartgun/Gameplay/NoLockOn/GE_NoLockOnSpread.GE_NoLockOnSpread_C" :Level 0.0))

(event Custom|TellServerReload
  (Variables|Default|SetReplicatedStacks 0)
  (Variables|Default|SetPowerupStacks 0))

(event EventTick
  (params (DeltaSeconds :type Float%20%28single-precision%29))
  (if (not (Networking|IsServer))
    (bind _asfort_pawn (Utilities|Casting|CastToFortPawn (Components|GetOwner)))
    (Variables|Default|SetOwnerPawn _asfort_pawn)
    (switch int (Variables|Default|GetPowerupStacks)
      (:0
        (Variables|Default|SetReticle%20Size 0.05))
      (:1
        (Variables|Default|SetReticle%20Size 0.038))
      (:2
        (Variables|Default|SetReticle%20Size 0.025))
      (:3
        (Variables|Default|SetReticle%20Size 0.01))
      (:4
        (Variables|Default|SetReticle%20Size 0.0075)))
    (CallFunction|TickAutoAim (Variables|Default|GetReticle%20Size))))

(fn GetStartupProgress
  (return (Math|Float|Clamp(Float) (Math|Float|SafeDivide (- (Utilities|Time|GetTimeSeconds) (Variables|Default|GetTargetingStartTime)) 1.0))))

(fn OnReload)

(fn OnRep_ReplicatedStacks
  (Development|PrintString (Utilities|String|ToString(Integer) (Variables|Default|GetReplicatedStacks))))

(fn OnTargetingStart
  (Variables|Default|SetStartingUp true)
  (Variables|Default|SetTargetingStartTime (Utilities|Time|GetTimeSeconds)))

(fn StopTargeting
  (Variables|Default|SetStartingUp false))

(fn TickAutoAim
  (params (ReticleSize :type Float%20%28double-precision%29)))
